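; Haunted House carpet prop that breaks when a pawn steps on it and handles audio for players entering the basement.
; On the server, the rug overlap sets b_EnteredBasement and destroys the actor; death effects play a wood-break sound after a short delay.
; For locally viewed players, kitchen, stairwell and shack triggers play a one-time landing stinger, attached to the kitchen trigger, gated by that flag.
; The mine trigger only sets the flag, suppressing later stingers.
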(event OnComponentBeginOverlap(OverlapTrigger_Rug)
  (params (OverlappedComponent :type Primitive%20Component%20Object%20Reference) (OtherActor :type Actor%20Object%20Reference) (OtherComp :type Primitive%20Component%20Object%20Reference) (OtherBodyIndex :type Integer) (bFromSweep :type Boolean) (SweepResult :type Hit%20Result%20Structure%20%28by%20ref%29))
  ;[1 1 1] Rug Entry Point
  (Networking|SwitchHasAuthority
    (:Authority
      (Utilities|Casting|CastToFortPawn OtherActor)
      (Variables|Default|Setb_EnteredBasement? true)
      (Actor|DestroyActor))))

(event EventOnDeathPlayEffects
  (params (Damage :type Float%20%28single-precision%29) (DamageTags :type Gameplay%20Tag%20Container%20Structure%20%28by%20ref%29) (Momentum :type Vector :default "0, 0, 0") (HitInfo :type Hit%20Result%20Structure%20%28by%20ref%29) (InstigatedBy :type Fort%20Pawn%20Object%20Reference) (DamageCauser :type Actor%20Object%20Reference) (EffectContext :type Gameplay%20Effect%20Context%20Handle%20Structure))
  ;[1 1 1] Rug Entry Point
  (Utilities|FlowControl|Delay 0.2) ; Extra Separation Between Break and Stinger from "BoxCollision_BasementFalling_Stinger_BP"
  (Audio|SpawnSoundAttached (Variables|Default|GetSFX_FloorBreak_Wood) (Variables|Debug|GetStaticMeshComponent)))

(event OnComponentBeginOverlap(OverlapTrigger_Kitchen_EntryPoint)
  (params (OverlappedComponent :type Primitive%20Component%20Object%20Reference) (OtherActor :type Actor%20Object%20Reference) (OtherComp :type Primitive%20Component%20Object%20Reference) (OtherBodyIndex :type Integer) (bFromSweep :type Boolean) (SweepResult :type Hit%20Result%20Structure%20%28by%20ref%29))
  ;[1 1 1] Kitchen Entry Point
  (Utilities|IsValid OtherActor
    (:Is%20Valid
      (bind _asplayer_pawn_athena (Utilities|Casting|CastToPlayerPawn_Athena OtherActor))
      (if (Pawn|IsLocallyViewed _asplayer_pawn_athena)
        (if (not (Variables|Default|Getb_EnteredBasement?))
          (Audio|SpawnSoundAttached (Variables|Default|GetSFX_Landing_Stinger) (Variables|Default|GetOverlapTrigger_Kitchen_EntryPoint)) ; Landing in Basement SFX Stinger (No Woodbreak)
          (Variables|Default|Setb_EnteredBasement? true))))))

(event OnComponentBeginOverlap(OverlapTrigger_Stairwell_EntryPoint)
  (params (OverlappedComponent :type Primitive%20Component%20Object%20Reference) (OtherActor :type Actor%20Object%20Reference) (OtherComp :type Primitive%20Component%20Object%20Reference) (OtherBodyIndex :type Integer) (bFromSweep :type Boolean) (SweepResult :type Hit%20Result%20Structure%20%28by%20ref%29))
  ;[1 1 1] Stairwell Entry Point
  (Utilities|IsValid OtherActor
    (:Is%20Valid
      (bind _asplayer_pawn_athena (Utilities|Casting|CastToPlayerPawn_Athena OtherActor))
      (if (Pawn|IsLocallyViewed _asplayer_pawn_athena)
        (if (not (Variables|Default|Getb_EnteredBasement?))
          (Audio|SpawnSoundAttached (Variables|Default|GetSFX_Landing_Stinger) (Variables|Default|GetOverlapTrigger_Kitchen_EntryPoint)) ; Landing in Basement SFX Stinger (No Woodbreak)
          (Variables|Default|Setb_EnteredBasement? true))))))

(event OnComponentBeginOverlap(OverlapTrigger_Shack_EntryPoint)
  (params (OverlappedComponent :type Primitive%20Component%20Object%20Reference) (OtherActor :type Actor%20Object%20Reference) (OtherComp :type Primitive%20Component%20Object%20Reference) (OtherBodyIndex :type Integer) (bFromSweep :type Boolean) (SweepResult :type Hit%20Result%20Structure%20%28by%20ref%29))
  ;[1 1 1] Shack Entry Point
  (Utilities|IsValid OtherActor
    (:Is%20Valid
      (bind _asplayer_pawn_athena (Utilities|Casting|CastToPlayerPawn_Athena OtherActor))
      (if (Pawn|IsLocallyViewed _asplayer_pawn_athena)
        (if (not (Variables|Default|Getb_EnteredBasement?))
          (Audio|SpawnSoundAttached (Variables|Default|GetSFX_Landing_Stinger) (Variables|Default|GetOverlapTrigger_Kitchen_EntryPoint)) ; Landing in Basement SFX Stinger (No Woodbreak)
          (Variables|Default|Setb_EnteredBasement? true))))))

(event OnComponentBeginOverlap(OverlapTrigger_Mine_EntryPoint)
  (params (OverlappedComponent :type Primitive%20Component%20Object%20Reference) (OtherActor :type Actor%20Object%20Reference) (OtherComp :type Primitive%20Component%20Object%20Reference) (OtherBodyIndex :type Integer) (bFromSweep :type Boolean) (SweepResult :type Hit%20Result%20Structure%20%28by%20ref%29))
  ;[1 1 1] Mine Entry Point
  (Utilities|IsValid OtherActor
    (:Is%20Valid
      (bind _asplayer_pawn_athena (Utilities|Casting|CastToPlayerPawn_Athena OtherActor))
      (if (Pawn|IsLocallyViewed _asplayer_pawn_athena)
        (if (not (Variables|Default|Getb_EnteredBasement?))
          (Variables|Default|Setb_EnteredBasement? true))))))

(fn ConstructionScript)
